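# EyeLaser_Temp is an unlit material that, by its name and Layout folder, appears to be a temporary red look for an eye laser.
# Its small graph multiplies a constant red color by a scalar parameter named Emissive, whose high default suggests a bright glow.
# The parameter lets material instances change the intensity, while the color stays fixed.
# Automatic usage-flag setting in the editor is turned off.
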Material /PistolCanary/Assets/Materials/Layout/EyeLaser_Temp
graph {
 properties {bAutomaticallySetUsageInEditor=false, shadingModel=MSM_Unlit, usedShadingModels=Unlit}
 nodes [Multiply(A=Constant3Vector() {constant=<rgba 1 0 0 0>}, B=ScalarParameter() {defaultValue=50, parameterName=Emissive})]
}
